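; Dynamic UI director component that shows the GreaseClipDunmurry dashboard scene to a local player who is driving a vehicle.
; Stateful gameplay events for started and should-hide, registered when UI is ready and released on EndPlay, toggle the dashboard and a show-HUD flag.
; Seat changes re-evaluate the scene using that flag.
; The toggle also hides the stock Valet vehicle HUD, keeping it hidden on touch, where Dynamic UI replaces that widget.
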(event EventOnReadyToShowUI
  (CallFunction|RegisterDonmurryStartedEvent)
  (CallFunction|RegisterDonmurryHideEvent)
  (bind _asfort_player_controller_athena (Utilities|Casting|CastToFortPlayerControllerAthena (CallFunction|GetOwningLocalPlayerController)))
  (Vehicle|BindEventtoOnVehicleSeatChanged _asfort_player_controller_athena (EventDispatchers|CreateEvent :!event OnVehicleSeatChanged)))

(event EventEndPlay
  (params (EndPlayReason :type EEndPlayReason%20Enum :default "Destroyed"))
  (CallFunction|UnregisterDonmurryStartedEvent)
  (CallFunction|UnregisterDonmurryHideEvent)
  (bind _asfort_player_controller_athena (Utilities|Casting|CastToFortPlayerControllerAthena (CallFunction|GetOwningLocalPlayerController)))
  (Vehicle|UnbindEventfromOnVehicleSeatChanged _asfort_player_controller_athena (EventDispatchers|CreateEvent :!event OnVehicleSeatChanged)))

(event Custom|RegisterDonmurryStartedEvent
  ;[1 1 1] Donmurry Started
  (bind (_asynctaskproxy _payload) (GameplayEventRouter|StartListeningToStatefulEvent (GameplayEventRouter|GetGlobalEventRouter) "/Script/GreaseClipDunmurryRuntime.OnGreaseClipDunmurryStarted" (CallFunction|GetOwningLocalPlayerController))
    (:then
      (Variables|Default|SetDunmurryStartedEventHandle _asynctaskproxy))
    (:OnEventReceived
      _)
    (:OnSavedEventStateExists
      _)
    (:OnEventStateCleared
      (CallFunction|AddorRemoveDashboard false)
      (Variables|Default|SetbShowGreaseClipHUD false)
      (Utilities|PostTransitionTouchCleanup)))
  (if (Utilities|Struct|BreakOnGreaseClipDunmurryStarted _payload)
    (CallFunction|AddorRemoveDashboard true)
    (Variables|Default|SetbShowGreaseClipHUD true)))

(event Custom|UnregisterDonmurryStartedEvent
  ;[1 1 1] Donmurry Started
  (GameplayEventRouter|StopListeningtoStatefulEvent (Variables|Default|GetDunmurryStartedEventHandle)))

(event Custom|RegisterDonmurryHideEvent
  ;[1 1 1] Donmurry Hide
  (bind (_asynctaskproxy _payload) (GameplayEventRouter|StartListeningToStatefulEvent (GameplayEventRouter|GetGlobalEventRouter) "/Script/GreaseClipDunmurryRuntime.OnGreaseClipDunmurryShouldHideUI" (CallFunction|GetOwningLocalPlayerController))
    (:then
      (Variables|Default|SetDonmurryHideEventHandle _asynctaskproxy))
    (:OnEventReceived
      (if (Utilities|Struct|BreakOnGreaseClipDunmurryShouldHideUI _payload)
        _
        (elif (.bWasCanceled (Utilities|Struct|BreakOnGreaseClipDunmurryShouldHideUI _payload))
          _
          (else
            (CallFunction|AddorRemoveDashboard false)
            (Variables|Default|SetbShowGreaseClipHUD false)))))))

(event Custom|UnregisterDonmurryHideEvent
  ;[1 1 1] Donmurry Hide
  (GameplayEventRouter|StopListeningtoStatefulEvent (Variables|Default|GetDonmurryHideEventHandle)))

(fn AddOrRemoveDashboard
  (params (bAdd :type Boolean))
  (bind _returnvalue (CallFunction|GetOwningFortDynamicUIDirector))
  (bind _returnvalue_2 (DynamicUIDirector|GetOwningLocalPlayerController _returnvalue))
  (bind _returnvalue_3 (CallFunction|GetOwningFortDynamicUIDirector))
  (bind _dashboard_scene (Variables|Default|GetDashboard%20Scene))
  (bind _isindriverseat (CallFunction|IsinDriverSeat))
  (LocalVariables|SetIsInDriverSeat _isindriverseat)
  (bind _asfort_player_controller (Utilities|Casting|CastToFortPlayerController _returnvalue_2))
  ;[1 1 1] Hide the original vehicle HUD (touch Replaces the current Widget using DynamicUI but needs to keep the KBM widget hidden)
  (HUD|SetHUDElementVisibilitywithContext _asfort_player_controller (GameplayTags|MakeLiteralGameplayTagContainer "(GameplayTags=((TagName=\"HUD.Athena.Valet\")))") (or (TouchPlatform|IsUsingTouch _returnvalue_2) (Variables|GetbAdd)) "None" "")
  (if (and (Variables|GetbAdd) (LocalVariables|GetIsInDriverSeat))
    (DynamicUIDirector|AddScene _returnvalue_3 _dashboard_scene)
    (else
      (DynamicUIDirector|RemoveScene _returnvalue_3 _dashboard_scene))))

(fn IsInDriverSeat
  (bind _asfort_player_pawn (Utilities|Casting|CastToFortPlayerPawn (Pawn|GetControlledPawn (DynamicUIDirector|GetOwningLocalPlayerController (CallFunction|GetOwningFortDynamicUIDirector))))
    (:then
      (return (Vehicle|IsDrivingVehicle _asfort_player_pawn)))
    (:CastFailed
      (return false))))

(fn OnVehicleSeatChanged
  (params (SeatIndex :type Integer))
  (CallFunction|AddorRemoveDashboard (and (== SeatIndex (Math|Conversions|ToInteger(Byte) (Utilities|Enum|LiteralenumEVehicleSeats "Driver"))) (Variables|Default|GetbShowGreaseClipHUD))))
